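; BP_PlayerPawn_LesserDemon_Boss is an AI alert-aware pawn Blueprint for a Lesser Demon boss that adds a combat-start speech bubble.
; Its AlertLevelChanged override calls the parent, then plays the LesserDemons CombatStart bubble through BPFLAICommonFunctions whenever the new alert level is Threatened.
; The construction script only calls the parent, and BeginPlay, Tick and overlap events are disabled placeholders.
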(event EventBeginPlay
  (|CallParentFunction) ; This node is disabled and will not be called.\nDrag off pins to build functionality.
)

(event Collision|EventActorBeginOverlap
  (params (OtherActor :type Actor%20Object%20Reference))
  (|CallParentFunction OtherActor) ; This node is disabled and will not be called.\nDrag off pins to build functionality.
)

(event EventTick
  (params (DeltaSeconds :type Float%20%28single-precision%29))
  (|CallParentFunction DeltaSeconds) ; This node is disabled and will not be called.\nDrag off pins to build functionality.
)

(fn AlertLevelChanged
  (params (OldAlertLevel :type EAlertLevel%20Enum :default "Unaware") (InputPin :type EAlertLevel%20Enum :default "Unaware"))
  (|CallParentFunction OldAlertLevel InputPin)
  (switch Utilities|FlowControl|Switch|SwitchonEAlertLevel InputPin
    (:Threatened
      (Class|BPFLAICommonFunctions|PlaySpeechBubble self "(TagName=\"Athena.AI.SpeechBubbles.LesserDemons.CombatStart\")"))))

(fn ConstructionScript
  (|CallParentFunction))
